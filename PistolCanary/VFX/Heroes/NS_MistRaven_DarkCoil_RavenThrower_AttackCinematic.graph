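# Niagara system for the cinematic attack of MistRaven's DarkCoil RavenThrower, staging a swirling raven flock that converges into an impact.
# Raven_Rotate circles ravens by vortex, delayed Raven_Attack pulls them inward and kills them near the center, and a later Impact emitter adds mesh and sprite effects.
# Raven emitters set custom trail and hit attributes for multiple renderers, and User.Ascended switches red to violet and enlarges sprites.
# Several emitters, including Minimal and Raven_Thrower, are disabled.
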NiagaraSystem /PistolCanary/VFX/Heroes/NS_MistRaven_DarkCoil_RavenThrower_AttackCinematic
dependencies {
 usedDataInterfaces [
  &NiagaraDataInterfaceCurve
  &NiagaraDataInterfaceVector2DCurve
  &NiagaraDataInterfaceVectorCurve
  &NiagaraDataInterfaceColorCurve
  &NiagaraDataInterfaceSpriteRendererInfo
 ]
 usedDynamicInputs [&InvertBool]
 usedModules [
  &self:SystemSpawnScript.SystemScriptSource.SystemScriptGraph.NiagaraNodeAssignment_1.SetVariables_#1
  &SystemState
  &self:SystemSpawnScript.SystemScriptSource.SystemScriptGraph.NiagaraNodeAssignment_0.SetVariables_#2
 ]
}
emitters {
 Impact {
  emitterUpdateScript [
   &EmitterState(Life Cycle Mode=::NewEnumerator1, Loop Behavior=::NewEnumerator1, Loop Delay=2.5, UseLoopDelay=true)
   &SpawnRate(SpawnRate=25, Use Spawn Probability=false)
  ]
  particleSpawnScript [
   &InitializeParticle(
    Color=<rgba 0.5 0 1.5 1>
    Lifetime Max=1
    Lifetime Min=0.5
    Lifetime Mode=::NewEnumerator1
    Mesh Scale Mode=::NewEnumerator1
    Mesh Uniform Scale Max=8
    Mesh Uniform Scale Min=2.5
    Sprite Rotation Mode=::NewEnumerator2
    Sprite Size Mode=::NewEnumerator1
    Sprite UV Mode=::NewEnumerator3
    Uniform Sprite Size Max=375
    Uniform Sprite Size Min=250
    UsePositionOffset=false
    Write Material Random=true
   )
   &ShapeLocation(
    Cylinder Height=3000
    Cylinder Radius=1000
    MaskSpawnGroup=false
    Offset Coordinate Space=::Local
    Offset Mode=::NewEnumerator0
    Override Seed=false
    Rotation Axis=<0 0 0>
    Rotation Mode=::NewEnumerator1
    Shape Primitive=::NewEnumerator1
    Surface Only=true
    Use Endcaps In Surface Only Mode=false
    UseOwnerScale=true
   )
   &CameraOffset(Camera Offset Amount=1000)
   &InitialMeshOrientation(Use Rotation Vector=false)
  ]
  particleUpdateScript [
   &ParticleState()
   &ScaleSpriteSize(
    Uniform Curve Scale=2
    Uniform Curve Sprite Scale=DataInterface() {propertyValues {Curve {
       defaultValue=0
       keys [
        {arriveTangent=5.8792, interpMode=RCIM_Cubic, leaveTangent=5.8792, tangentMode=RCTM_User, time=0, value=0}
        {arriveTangent=0.0224, interpMode=RCIM_Cubic, leaveTangent=0.0224, tangentMode=RCTM_User, time=0.5, value=1}
       ]
      }}}
   )
   &ScaleMeshSize(Scale Factor=&VectorFromFloat(Value=&FloatFromCurve(
      FloatCurve=DataInterface() {propertyValues {Curve {
         defaultValue=0
         keys [
          {interpMode=RCIM_Cubic, tangentMode=RCTM_User, time=0, value=0.5}
          {interpMode=RCIM_Cubic, tangentMode=RCTM_User, time=0.1, value=1}
          <at 0.5 0 RCIM_Cubic>
         ]
        }}}
      Scale Curve=2
     )))
   &DynamicMaterialParameters(
    Index 0 Param 1=&FloatFromCurve(FloatCurve=DataInterface() {propertyValues {Curve {defaultValue=1, keys [{tangentMode=RCTM_User, time=0, value=0}, {tangentMode=RCTM_User, time=1, value=1}]}}})
    Module.Index0Float0Write=true
    Module.Index0Float1Write=false
    Module.Index0Float2Write=false
    Module.Index0Float3Write=false
   )
   &SubUVAnimation(
    "Sprite Renderer "=DataInterface() {propertyValues {SpriteRenderer=&self:Impact_0.NiagaraSpriteRendererProperties_1}}
    UseEndFrame=false
    UseStartFrame=false
   )
  ]
  rendererClasses [&NiagaraMeshRendererProperties, &NiagaraSpriteRendererProperties]
  renderers [
   {rendererClass=&NiagaraMeshRendererProperties, rendererIndex=0}
   {rendererClass=&NiagaraSpriteRendererProperties, rendererIndex=1}
   {rendererClass=&NiagaraMeshRendererProperties, rendererIndex=2}
  ]
 }
 Minimal {
  bEnabled=false
  emitterUpdateScript [
   &EmitterState(Life Cycle Mode=::NewEnumerator1, Loop Behavior=::NewEnumerator1, Loop Delay=2.75, UseLoopDelay=true)
   &SpawnRate(SpawnRate=5, Use Spawn Probability=false)
  ]
  particleSpawnScript [
   &InitializeParticle(
    Color=<rgba 0.5 0 1.5 100>
    Lifetime=10
    Sprite Rotation Mode=::NewEnumerator2
    Sprite Size=<50 5000>
    Sprite Size Mode=::NewEnumerator4
    UsePositionOffset=false
    Write Material Random=true
   )
   &ShapeLocation(MaskSpawnGroup=false, Non Uniform Scale=<1 1 0.25>, Override Seed=false, Sphere Radius=5000, UseOwnerScale=true) {module=ShapeLocation001}
   &ShapeLocation(
    Cylinder Height=5000
    Cylinder Radius=4000
    MaskSpawnGroup=false
    Non Uniform Scale=<1 1 0.5>
    Offset Coordinate Space=::Local
    Offset Mode=::NewEnumerator0
    Override Seed=false
    Rotation Axis=<0 0 0>
    Rotation Mode=::NewEnumerator1
    Shape Primitive=::NewEnumerator1
    Surface Only=true
    Use Endcaps In Surface Only Mode=false
    UseOwnerScale=true
   ) {enabled=false}
   &CameraOffset(Camera Offset Amount=2500) {enabled=false}
  ]
  particleUpdateScript [
   &ParticleState()
   &PointForce(Force Strength=-100, Module.Use Falloff=false, Module.UseLinearForce=false)
   &SolveForcesAndVelocity(Clamp Velocity=false, Limit Acceleration=false)
   &ScaleColor(
    Scale Alpha=&FloatFromCurve(FloatCurve=DataInterface() {propertyValues {Curve {keys [<at 0 1>,<at 1 0>]}}}, CurveIndex=$Emitter.Age)
    ScaleA=true
    ScaleRGB=true
   )
  ]
  rendererClasses [&NiagaraSpriteRendererProperties]
  renderers [{rendererClass=&NiagaraSpriteRendererProperties}]
 }
 Raven_Attack {
  emitterUpdateScript [
   &EmitterState(Life Cycle Mode=::NewEnumerator1, Loop Behavior=::NewEnumerator1, Loop Delay=2, UseLoopDelay=true)
   &SpawnRate(SpawnRate=50, Use Spawn Probability=false) {module=SpawnRate001}
  ]
  particleSpawnScript [
   &InitializeParticle(
    Color=&MakeCustomLinearColorFromBool(Bool=$User.Ascended, True Color=<rgba 0.7481 0 2.5 1>, False Color=<rgba 5 0 0.2 1>)
    Lifetime=0.5
    Mesh Scale Mode=::NewEnumerator3
    Mesh Uniform Scale=&RandomRangeFloat(Minimum=1, Maximum=1.5)
    Sprite Size Mode=::NewEnumerator3
    Uniform Sprite Size=&MakeCustomFloatFromBool(Bool=$User.Ascended, True Float=250, False Float=150)
    UsePositionOffset=false
    Write Material Random=true
   )
   &ShapeLocation(
    Cylinder Height=1250
    Cylinder Radius=4000
    MaskSpawnGroup=false
    Non Uniform Scale=<1 1 0.5>
    Offset Coordinate Space=::Local
    Offset Mode=::NewEnumerator0
    Override Seed=false
    Rotation Axis=<0 0 0>
    Rotation Mode=::NewEnumerator1
    Shape Primitive=::NewEnumerator1
    Surface Only=true
    Use Endcaps In Surface Only Mode=false
    UseOwnerScale=true
   )
   &InitialMeshOrientation(Mesh Orientation Mode=::NewEnumerator2, Use Rotation Vector=false)
   &CameraOffset(Camera Offset Amount=-25)
   /SetVariables(Particles.CamOfs2=30) {isSetParameters=true, module=SetVariables_#3}
  ]
  particleUpdateScript [
   &ParticleState()
   &ScaleSpriteSize(
    Non-Uniform Curve Scale=<1.5 1>
    Non-Uniform Curve Sprite Scale=DataInterface() {propertyValues {
      XCurve {keys [<at 0.05 0 RCIM_Cubic>, {tangentMode=RCTM_User, time=0.3, value=1}, <at 0.5 1>, <at 1 0>]}
      YCurve {keys [<at 0.05 0 RCIM_Cubic>, {tangentMode=RCTM_User, time=0.3, value=1}, <at 0.5 1>, <at 1 0>]}
     }}
    Scale Sprite Size Mode=::NewEnumerator3
   )
   &ScaleMeshSize(Scale Factor=&Multiply_Vector(
     A=&VectorFromFloat(Value=&FloatFromCurve(
       FloatCurve=DataInterface() {propertyValues {Curve {
          defaultValue=1
          keys [
           <at 0 0 RCIM_Cubic>
           {arriveTangent=-0.1877, interpMode=RCIM_Cubic, leaveTangent=-0.1877, tangentMode=RCTM_User, time=0.5, value=1}
           {arriveTangent=0.0991, interpMode=RCIM_Cubic, leaveTangent=0.0991, tangentMode=RCTM_User, time=0.9, value=1}
           <at 1 0 RCIM_Cubic>
          ]
         }}}
       Scale Curve=$System.Timeline.InverseNormalizedTime
      ))
     B=&VectorFromCurve(VectorCurve=DataInterface() {propertyValues {XCurve {keys [<at 0.1 0>,<at 0.2 1>,<at 1 1>]}, YCurve {keys [<at 0 1>,<at 1 1>]}, ZCurve {keys [<at 0 1>,<at 1 1>]}}})
    ))
   &Drag(Drag=&RandomRangeFloat(Minimum=0.5, Maximum=10), Use Linear Drag=true, Use Rotational Drag=false)
   &CurlNoiseForce(Bias Noise=false, Noise Quality / Cost=::NewEnumerator0, Noise Strength=500, Pan Noise=false)
   &DynamicMaterialParameters(
    Index 1 Param 1=&FloatFromCurve(FloatCurve=DataInterface() {propertyValues {Curve {keys [<at 0 1>,<at 0.3 0>]}}})
    Index 1 Param 2=&RandomRangeFloat(Minimum=2, Maximum=3)
    Index 1 Param 3=&FloatFromCurve(
     FloatCurve=DataInterface() {propertyValues {Curve {keys [<at 0.2 5>,<at 0.3 5>,<at 1 0.5>]}}}
     Scale Curve=&FloatFromCurve(
      FloatCurve=DataInterface() {propertyValues {Curve {defaultValue=1, keys [{tangentMode=RCTM_User, time=0, value=0.3}, {tangentMode=RCTM_User, time=0.15, value=1}]}}}
      CurveIndex=$Particles.Age
     )
    )
    Index 1 Param 4=-1
    Index 2 Param 1=&RandomRangeFloat(Minimum=0, Maximum=1111)
    Index 3 Param 1=&RandomRangeFloat(Minimum=-0.5, Maximum=3)
    Index 3 Param 2=&RandomRangeFloat(Minimum=0, Maximum=1)
    Module.Index0Float0Write=true
    Module.Index0Float1Write=true
    Module.Index0Float2Write=false
    Module.Index0Float3Write=true
    Module.Index1Float0Write=true
    Module.Index1Float1Write=true
    Module.Index1Float2Write=true
    Module.Index1Float3Write=true
    Module.Index2Float0Write=true
    Module.Index2Float1Write=false
    Module.Index2Float2Write=false
    Module.Index2Float3Write=false
    Module.Index3Float0Write=true
    Module.Index3Float1Write=true
    Module.Index3Float2Write=true
    Module.Index3Float3Write=true
    Write Parameter Index 1=::NewEnumerator1
    Write Parameter Index 2=::NewEnumerator1
    Write Parameter Index 3=::NewEnumerator1
   )
   /SetVariables(Particles.ColorBlack:LinearColor=&MakeLinearColorFromVectorAndFloat(
     "Vector (RGB)"=&MakeVectorFromLinearColorRGB(Color=&ColorFromCurve(ColorCurve=DataInterface() {propertyValues {
         AlphaCurve {keys [<at 0 0 RCIM_Cubic>]}
         BlueCurve {keys [<at 0.745 0 RCIM_Cubic>]}
         GreenCurve {keys [<at 0.745 0 RCIM_Cubic>]}
         RedCurve {keys [<at 0.745 0 RCIM_Cubic>]}
        }}))
     "Float (Alpha)"=&MakeCustomFloatFromBool(
      Bool=$Particles.HasCollided
      True Float=0
      False Float=&FloatFromCurve(FloatCurve=DataInterface() {propertyValues {Curve {keys [<at 0.55 1>,<at 1 0>]}}}, Scale Curve=2)
     )
    )) {isSetParameters=true, module=SetVariables_#4}
   /SetVariables(
    Particles.DMP_Trail:Vector4f=&MakeVector4(
     X=0
     Y=-0.1
     Z=1
     W=&FloatFromCurve(FloatCurve=DataInterface() {propertyValues {Curve {defaultValue=1, keys [{tangentMode=RCTM_User, time=0.05, value=0}, {tangentMode=RCTM_User, time=1, value=1}]}}})
    )
    Particles.DMP_Trail1:Vector4f=&MakeVector4(X=1, Y=1, Z=1, W=1)
    Particles.DMP_Trail2:Vector4f=&MakeVector4(X=1, Y=1, Z=1, W=1)
    Particles.DMP_Trail3:Vector4f=&MakeVector4(X=500, Y=25, Z=0, W=0)
    Particles.Trail Color:LinearColor=&MultiplyLinearColor(
     Linear Color=&MakeCustomLinearColorFromBool(Bool=$User.Ascended, True Color=<rgba 2 0 0.9407 1>, False Color=<rgba 0.9219 0.024 0.024 1>)
     Scale Factor=&MakeVector4(
      X=1
      Y=1
      Z=1
      W=&MakeCustomFloatFromBool(
       Bool=$Particles.HasCollided
       True Float=0
       False Float=&FloatFromCurve(FloatCurve=DataInterface() {propertyValues {Curve {keys [<at 0.9 1>,<at 1 0>]}}})
      )
     )
    )
    Particles.Trail Orientation:Quat4f=&MakeQuaternion(
     Quaternion From=0
     Axis=<1 0 0>
     Angle=&Add_Float(
      A=&RandomRangeFloat(Minimum=0, Maximum=360)
      B=&Multiply_Float(A=$Engine.Time, B=1000)
     )
     From=?
     To=?
     Angle Type=::NewEnumerator0
     Normalize Vectors=false
    )
    Particles.Trail Scale:Vector3f=&VectorFromCurve(VectorCurve=DataInterface() {propertyValues {
       XCurve {defaultValue=1, keys [{tangentMode=RCTM_User, time=0, value=1.25}, {tangentMode=RCTM_User, time=0.2, value=2.375}, <at 1 0.5>]}
       YCurve {defaultValue=1, keys [{tangentMode=RCTM_User, time=0, value=0.25}, {tangentMode=RCTM_User, time=0.2, value=1}, <at 1 2>]}
       ZCurve {defaultValue=1, keys [{tangentMode=RCTM_User, time=0, value=0.25}, {tangentMode=RCTM_User, time=0.2, value=1}, <at 1 2>]}
      }})
    Particles.Velocity=&MakeCustomVectorFromBool(Bool=$Particles.HasCollided, True Vector=<0 0 0>, False Vector=$Particles.Velocity)
   ) {isSetParameters=true, module=SetVariables_#5}
   &Timeline(Play=$Particles.HasCollided, Play Rate=2)
   &ScaleColor(
    Scale RGB=&Multiply_VectorByFloat(
     Vector=&MakeVectorFromLinearColorRGB(Color=&ColorFromCurve(ColorCurve=DataInterface() {propertyValues {
         AlphaCurve {keys [<at 0 0.2>, {arriveTangent=-0.2, interpMode=RCIM_Cubic, leaveTangent=-0.2, time=0.1889, value=1}, <at 1 0 RCIM_Cubic>]}
         BlueCurve {keys [<at 0 1 RCIM_Cubic>]}
         GreenCurve {keys [<at 0 1 RCIM_Cubic>]}
         RedCurve {keys [<at 0 1 RCIM_Cubic>]}
        }}))
     Float=&MakeCustomFloatFromBool(Bool=$Particles.HasCollided, True Float=0, False Float=1)
    )
    ScaleA=false
    ScaleRGB=true
   )
   /SetVariables(
    Particles.Hit Color:LinearColor=&MakeLinearColorFromVectorAndFloat(
     "Vector (RGB)"=&MakeCustomVectorFromBool(
      Bool=$User.Ascended
      True Vector=&VectorFromCurve(
       VectorCurve=DataInterface() {propertyValues {
         XCurve {keys [{arriveTangent=-3.5433, time=0, value=0.625}, <at 0.5 0.1>]}
         YCurve {keys [<at 0.01 0.1>,<at 0.5 0>]}
         ZCurve {keys [<at 0 1.375>,<at 0.5 0.375>]}
        }}
       CurveIndex=$Particles.Timeline.NormalizedTime
      )
      False Vector=&VectorFromCurve(
       VectorCurve=DataInterface() {propertyValues {
         XCurve {keys [{arriveTangent=-12.0139, interpMode=RCIM_Cubic, leaveTangent=-12.0139, tangentMode=RCTM_User, time=0.01, value=2}, <at 0.5 0>]}
         YCurve {keys [<at 0.01 0.1>,<at 0.5 0>]}
         ZCurve {keys [<at 0.01 0.1>,<at 0.5 0>]}
        }}
       CurveIndex=$Particles.Timeline.NormalizedTime
      )
     )
     "Float (Alpha)"=&Multiply_Float(A=5, B=$Particles.Timeline.InverseNormalizedTime)
    )
    Particles.Hit Sprite:Vector2f=&Vector2DFromCurve(
     Vector2Curve=DataInterface() {propertyValues {
       XCurve {
        defaultValue=1
        keys [
         {arriveTangent=4.1785, interpMode=RCIM_Cubic, leaveTangent=4.1785, tangentMode=RCTM_User, time=0.01, value=0}
         {arriveTangent=1.1179, interpMode=RCIM_Cubic, leaveTangent=1.1179, tangentMode=RCTM_User, time=0.3, value=0.65}
         {arriveTangent=0.0224, interpMode=RCIM_Cubic, leaveTangent=0.0224, tangentMode=RCTM_User, time=1, value=1}
        ]
       }
       YCurve {
        defaultValue=1
        keys [
         {arriveTangent=4.1785, interpMode=RCIM_Cubic, leaveTangent=4.1785, tangentMode=RCTM_User, time=0.01, value=0}
         {arriveTangent=1.1179, interpMode=RCIM_Cubic, leaveTangent=1.1179, tangentMode=RCTM_User, time=0.3, value=0.65}
         {arriveTangent=0.0224, interpMode=RCIM_Cubic, leaveTangent=0.0224, tangentMode=RCTM_User, time=1, value=1}
        ]
       }
      }}
     CurveIndex=$Particles.Timeline.NormalizedTime
     ScaleCurve=<120 120>
    )
   ) {isSetParameters=true, module=SetVariables_#6}
   &SubUVAnimation(
    Random Start Frame=true
    "Sprite Renderer "=DataInterface() {propertyValues {SpriteRenderer=&self:Raven_Attack_0.NiagaraSpriteRendererProperties_1}}
    SubUV Lookup Index:NiagaraFloat=$StackContext.Timeline.NormalizedTime
    UseEndFrame=false
    UseStartFrame=false
   )
   &VortexForce(
    Origin Pull Amount=5000
    UseOriginOffset=false
    Vortex Axis=<0 0 0>
    Vortex Force Amount=&RandomRangeFloat(Minimum=10000, Maximum=50000)
   )
   &PointAttractionForce(
    Attraction Radius=50000
    Attraction Strength=&RandomRangeFloat(Minimum=500, Maximum=1000)
    Kill Radius=1000
    Kill Within Radius=true
    Module.Use Falloff=true
    UseAttractorOffset=false
   ) {module=PointAttractionForce001}
   &SolveForcesAndVelocity(Clamp Velocity=false, Limit Acceleration=false)
  ]
  rendererClasses [&NiagaraMeshRendererProperties, &NiagaraSpriteRendererProperties]
  renderers [
   {rendererClass=&NiagaraMeshRendererProperties, rendererIndex=0}
   {rendererClass=&NiagaraSpriteRendererProperties, rendererIndex=1}
   {rendererClass=&NiagaraMeshRendererProperties, rendererIndex=2}
   {rendererClass=&NiagaraSpriteRendererProperties, rendererIndex=3}
   {rendererClass=&NiagaraMeshRendererProperties, rendererIndex=4}
  ]
 }
 Raven_Attack001 {
  bEnabled=false
  emitterUpdateScript [
   &EmitterState(Life Cycle Mode=::NewEnumerator1, Loop Behavior=::NewEnumerator1, Loop Delay=2, UseLoopDelay=true)
   &SpawnRate(SpawnRate=50, Use Spawn Probability=false) {module=SpawnRate001}
  ]
  particleSpawnScript [
   &InitializeParticle(
    Color=&MakeCustomLinearColorFromBool(Bool=$User.Ascended, True Color=<rgba 0.7481 0 2.5 1>, False Color=<rgba 5 0 0.2 1>)
    Lifetime=0.5
    Mesh Scale Mode=::NewEnumerator3
    Mesh Uniform Scale=&RandomRangeFloat(Minimum=1, Maximum=1.5)
    Sprite Size Mode=::NewEnumerator3
    Uniform Sprite Size=&MakeCustomFloatFromBool(Bool=$User.Ascended, True Float=250, False Float=150)
    UsePositionOffset=false
    Write Material Random=true
   )
   &ShapeLocation(
    Cylinder Height=1250
    Cylinder Radius=4000
    MaskSpawnGroup=false
    Non Uniform Scale=<1 1 0.5>
    Offset Coordinate Space=::Local
    Offset Mode=::NewEnumerator0
    Override Seed=false
    Rotation Axis=<0 0 0>
    Rotation Mode=::NewEnumerator1
    Shape Primitive=::NewEnumerator1
    Surface Only=false
    UseOwnerScale=true
   )
   &InitialMeshOrientation(Mesh Orientation Mode=::NewEnumerator2, Use Rotation Vector=false)
   &CameraOffset(Camera Offset Amount=-25)
   /SetVariables(Particles.CamOfs2=30) {isSetParameters=true, module=SetVariables_#3}
  ]
  particleUpdateScript [
   &ParticleState()
   &ScaleSpriteSize(
    Non-Uniform Curve Scale=<1.5 1>
    Non-Uniform Curve Sprite Scale=DataInterface() {propertyValues {
      XCurve {keys [<at 0.05 0 RCIM_Cubic>, {tangentMode=RCTM_User, time=0.3, value=1}, <at 0.5 1>, <at 1 0>]}
      YCurve {keys [<at 0.05 0 RCIM_Cubic>, {tangentMode=RCTM_User, time=0.3, value=1}, <at 0.5 1>, <at 1 0>]}
     }}
    Scale Sprite Size Mode=::NewEnumerator3
   )
   &ScaleMeshSize(Scale Factor=&Multiply_Vector(
     A=&VectorFromFloat(Value=&FloatFromCurve(
       FloatCurve=DataInterface() {propertyValues {Curve {
          defaultValue=1
          keys [
           <at 0 0 RCIM_Cubic>
           {arriveTangent=-0.1877, interpMode=RCIM_Cubic, leaveTangent=-0.1877, tangentMode=RCTM_User, time=0.5, value=1}
           {arriveTangent=0.0991, interpMode=RCIM_Cubic, leaveTangent=0.0991, tangentMode=RCTM_User, time=0.9, value=1}
           <at 1 0 RCIM_Cubic>
          ]
         }}}
       Scale Curve=$System.Timeline.InverseNormalizedTime
      ))
     B=&VectorFromCurve(VectorCurve=DataInterface() {propertyValues {XCurve {keys [<at 0.1 0>,<at 0.2 1>,<at 1 1>]}, YCurve {keys [<at 0 1>,<at 1 1>]}, ZCurve {keys [<at 0 1>,<at 1 1>]}}})
    ))
   &Drag(Drag=&RandomRangeFloat(Minimum=0.5, Maximum=10), Use Linear Drag=true, Use Rotational Drag=false)
   &CurlNoiseForce(Bias Noise=false, Noise Quality / Cost=::NewEnumerator0, Noise Strength=500, Pan Noise=false)
   &DynamicMaterialParameters(
    Index 1 Param 1=&FloatFromCurve(FloatCurve=DataInterface() {propertyValues {Curve {keys [<at 0 1>,<at 0.3 0>]}}})
    Index 1 Param 2=&RandomRangeFloat(Minimum=2, Maximum=3)
    Index 1 Param 3=&FloatFromCurve(
     FloatCurve=DataInterface() {propertyValues {Curve {keys [<at 0.2 5>,<at 0.3 5>,<at 1 0.5>]}}}
     Scale Curve=&FloatFromCurve(
      FloatCurve=DataInterface() {propertyValues {Curve {defaultValue=1, keys [{tangentMode=RCTM_User, time=0, value=0.3}, {tangentMode=RCTM_User, time=0.15, value=1}]}}}
      CurveIndex=$Particles.Age
     )
    )
    Index 1 Param 4=-1
    Index 2 Param 1=&RandomRangeFloat(Minimum=0, Maximum=1111)
    Index 3 Param 1=&RandomRangeFloat(Minimum=-0.5, Maximum=3)
    Index 3 Param 2=&RandomRangeFloat(Minimum=0, Maximum=1)
    Module.Index0Float0Write=true
    Module.Index0Float1Write=true
    Module.Index0Float2Write=false
    Module.Index0Float3Write=true
    Module.Index1Float0Write=true
    Module.Index1Float1Write=true
    Module.Index1Float2Write=true
    Module.Index1Float3Write=true
    Module.Index2Float0Write=true
    Module.Index2Float1Write=false
    Module.Index2Float2Write=false
    Module.Index2Float3Write=false
    Module.Index3Float0Write=true
    Module.Index3Float1Write=true
    Module.Index3Float2Write=true
    Module.Index3Float3Write=true
    Write Parameter Index 1=::NewEnumerator1
    Write Parameter Index 2=::NewEnumerator1
    Write Parameter Index 3=::NewEnumerator1
   )
   /SetVariables(Particles.ColorBlack:LinearColor=&MakeLinearColorFromVectorAndFloat(
     "Vector (RGB)"=&MakeVectorFromLinearColorRGB(Color=&ColorFromCurve(ColorCurve=DataInterface() {propertyValues {
         AlphaCurve {keys [<at 0 0 RCIM_Cubic>]}
         BlueCurve {keys [<at 0.745 0 RCIM_Cubic>]}
         GreenCurve {keys [<at 0.745 0 RCIM_Cubic>]}
         RedCurve {keys [<at 0.745 0 RCIM_Cubic>]}
        }}))
     "Float (Alpha)"=&MakeCustomFloatFromBool(
      Bool=$Particles.HasCollided
      True Float=0
      False Float=&FloatFromCurve(FloatCurve=DataInterface() {propertyValues {Curve {keys [<at 0.55 1>,<at 1 0>]}}}, Scale Curve=2)
     )
    )) {isSetParameters=true, module=SetVariables_#4}
   /SetVariables(
    Particles.DMP_Trail:Vector4f=&MakeVector4(
     X=0
     Y=-0.1
     Z=1
     W=&FloatFromCurve(FloatCurve=DataInterface() {propertyValues {Curve {defaultValue=1, keys [{tangentMode=RCTM_User, time=0.05, value=0}, {tangentMode=RCTM_User, time=1, value=1}]}}})
    )
    Particles.DMP_Trail1:Vector4f=&MakeVector4(X=1, Y=1, Z=1, W=1)
    Particles.DMP_Trail2:Vector4f=&MakeVector4(X=1, Y=1, Z=1, W=1)
    Particles.DMP_Trail3:Vector4f=&MakeVector4(X=10000, Y=25, Z=0, W=0)
    Particles.Trail Color:LinearColor=&MultiplyLinearColor(
     Linear Color=&MakeCustomLinearColorFromBool(Bool=$User.Ascended, True Color=<rgba 2 0 0.9407 1>, False Color=<rgba 0.9219 0.024 0.024 1>)
     Scale Factor=&MakeVector4(
      X=1
      Y=1
      Z=1
      W=&MakeCustomFloatFromBool(
       Bool=$Particles.HasCollided
       True Float=0
       False Float=&FloatFromCurve(FloatCurve=DataInterface() {propertyValues {Curve {keys [<at 0.9 1>,<at 1 0>]}}}, Scale Curve=5)
      )
     )
    )
    Particles.Trail Orientation:Quat4f=&MakeQuaternion(
     Quaternion From=0
     Axis=<1 0 0>
     Angle=&Add_Float(
      A=&RandomRangeFloat(Minimum=0, Maximum=360)
      B=&Multiply_Float(A=$Engine.Time, B=1000)
     )
     From=?
     To=?
     Angle Type=::NewEnumerator0
     Normalize Vectors=false
    )
    Particles.Trail Scale:Vector3f=&VectorFromCurve(VectorCurve=DataInterface() {propertyValues {
       XCurve {defaultValue=1, keys [{tangentMode=RCTM_User, time=0, value=1.25}, {tangentMode=RCTM_User, time=0.2, value=2.375}, <at 1 0.5>]}
       YCurve {defaultValue=1, keys [{tangentMode=RCTM_User, time=0, value=0.25}, {tangentMode=RCTM_User, time=0.2, value=1}, <at 1 2>]}
       ZCurve {defaultValue=1, keys [{tangentMode=RCTM_User, time=0, value=0.25}, {tangentMode=RCTM_User, time=0.2, value=1}, <at 1 2>]}
      }})
    Particles.Velocity=&MakeCustomVectorFromBool(Bool=$Particles.HasCollided, True Vector=<0 0 0>, False Vector=$Particles.Velocity)
   ) {isSetParameters=true, module=SetVariables_#5}
   &Timeline(Play=$Particles.HasCollided, Play Rate=2)
   &ScaleColor(
    Scale RGB=&Multiply_VectorByFloat(
     Vector=&MakeVectorFromLinearColorRGB(Color=&ColorFromCurve(ColorCurve=DataInterface() {propertyValues {
         AlphaCurve {keys [<at 0 0.2>, {arriveTangent=-0.2, interpMode=RCIM_Cubic, leaveTangent=-0.2, time=0.1889, value=1}, <at 1 0 RCIM_Cubic>]}
         BlueCurve {keys [<at 0 1 RCIM_Cubic>]}
         GreenCurve {keys [<at 0 1 RCIM_Cubic>]}
         RedCurve {keys [<at 0 1 RCIM_Cubic>]}
        }}))
     Float=&MakeCustomFloatFromBool(Bool=$Particles.HasCollided, True Float=0, False Float=1)
    )
    ScaleA=false
    ScaleRGB=true
   )
   /SetVariables(
    Particles.Hit Color:LinearColor=&MakeLinearColorFromVectorAndFloat(
     "Vector (RGB)"=&MakeCustomVectorFromBool(
      Bool=$User.Ascended
      True Vector=&VectorFromCurve(
       VectorCurve=DataInterface() {propertyValues {
         XCurve {keys [{arriveTangent=-3.5433, time=0, value=0.625}, <at 0.5 0.1>]}
         YCurve {keys [<at 0.01 0.1>,<at 0.5 0>]}
         ZCurve {keys [<at 0 1.375>,<at 0.5 0.375>]}
        }}
       CurveIndex=$Particles.Timeline.NormalizedTime
      )
      False Vector=&VectorFromCurve(
       VectorCurve=DataInterface() {propertyValues {
         XCurve {keys [{arriveTangent=-12.0139, interpMode=RCIM_Cubic, leaveTangent=-12.0139, tangentMode=RCTM_User, time=0.01, value=2}, <at 0.5 0>]}
         YCurve {keys [<at 0.01 0.1>,<at 0.5 0>]}
         ZCurve {keys [<at 0.01 0.1>,<at 0.5 0>]}
        }}
       CurveIndex=$Particles.Timeline.NormalizedTime
      )
     )
     "Float (Alpha)"=&Multiply_Float(A=5, B=$Particles.Timeline.InverseNormalizedTime)
    )
    Particles.Hit Sprite:Vector2f=&Vector2DFromCurve(
     Vector2Curve=DataInterface() {propertyValues {
       XCurve {
        defaultValue=1
        keys [
         {arriveTangent=4.1785, interpMode=RCIM_Cubic, leaveTangent=4.1785, tangentMode=RCTM_User, time=0.01, value=0}
         {arriveTangent=1.1179, interpMode=RCIM_Cubic, leaveTangent=1.1179, tangentMode=RCTM_User, time=0.3, value=0.65}
         {arriveTangent=0.0224, interpMode=RCIM_Cubic, leaveTangent=0.0224, tangentMode=RCTM_User, time=1, value=1}
        ]
       }
       YCurve {
        defaultValue=1
        keys [
         {arriveTangent=4.1785, interpMode=RCIM_Cubic, leaveTangent=4.1785, tangentMode=RCTM_User, time=0.01, value=0}
         {arriveTangent=1.1179, interpMode=RCIM_Cubic, leaveTangent=1.1179, tangentMode=RCTM_User, time=0.3, value=0.65}
         {arriveTangent=0.0224, interpMode=RCIM_Cubic, leaveTangent=0.0224, tangentMode=RCTM_User, time=1, value=1}
        ]
       }
      }}
     CurveIndex=$Particles.Timeline.NormalizedTime
     ScaleCurve=<120 120>
    )
   ) {isSetParameters=true, module=SetVariables_#6}
   &PointAttractionForce(
    Attraction Radius=5000
    Attraction Strength=&RandomRangeFloat(Minimum=500, Maximum=1000)
    Kill Radius=1000
    Kill Within Radius=false
    Module.Use Falloff=true
    UseAttractorOffset=false
   ) {module=PointAttractionForce001}
   &SolveForcesAndVelocity(Clamp Velocity=false, Limit Acceleration=false)
  ]
  rendererClasses [&NiagaraMeshRendererProperties]
  renderers [{rendererClass=&NiagaraMeshRendererProperties, rendererIndex=0}]
 }
 Raven_Rotate {
  emitterUpdateScript [
   &EmitterState()
   &SpawnRate(SpawnRate=50, Use Spawn Probability=false) {module=SpawnRate001}
   &SpawnRate(
    SpawnRate=&Multiply_Float(
     A=30
     B=&FloatFromCurve(FloatCurve=DataInterface() {propertyValues {Curve {keys [<at 0.5 1 RCIM_Cubic>,<at 1 0 RCIM_Cubic>]}}}, CurveIndex=$System.NormalizedLoopAge)
    )
    Use Spawn Probability=false
   ) {enabled=false}
  ]
  particleSpawnScript [
   &InitializeParticle(
    Color=&MakeCustomLinearColorFromBool(Bool=$User.Ascended, True Color=<rgba 0.7481 0 2.5 1>, False Color=<rgba 5 0 0.2 1>)
    Lifetime=&RandomRangeFloat(Minimum=2, Maximum=4)
    Mesh Scale Mode=::NewEnumerator3
    Mesh Uniform Scale=&RandomRangeFloat(Minimum=1, Maximum=1.5)
    Sprite Size Mode=::NewEnumerator3
    Uniform Sprite Size=&MakeCustomFloatFromBool(Bool=$User.Ascended, True Float=250, False Float=150)
    UsePositionOffset=false
    Write Material Random=true
   )
   &ShapeLocation(
    Cylinder Height=1250
    Cylinder Radius=5000
    MaskSpawnGroup=false
    Non Uniform Scale=<1 1 0.5>
    Offset Coordinate Space=::Local
    Offset Mode=::NewEnumerator0
    Override Seed=false
    Rotation Axis=<0 0 0>
    Rotation Mode=::NewEnumerator1
    Shape Primitive=::NewEnumerator1
    Surface Only=true
    Use Endcaps In Surface Only Mode=false
    UseOwnerScale=true
   )
   &InitialMeshOrientation(Mesh Orientation Mode=::NewEnumerator2, Use Rotation Vector=false)
   &CameraOffset(Camera Offset Amount=-25)
   /SetVariables(Particles.CamOfs2=30) {isSetParameters=true, module=SetVariables_#3}
  ]
  particleUpdateScript [
   &ParticleState()
   &ScaleSpriteSize(
    Non-Uniform Curve Scale=<1.5 1>
    Non-Uniform Curve Sprite Scale=DataInterface() {propertyValues {
      XCurve {keys [<at 0.05 0 RCIM_Cubic>, {tangentMode=RCTM_User, time=0.3, value=1}, <at 0.5 1>, <at 1 0>]}
      YCurve {keys [<at 0.05 0 RCIM_Cubic>, {tangentMode=RCTM_User, time=0.3, value=1}, <at 0.5 1>, <at 1 0>]}
     }}
    Scale Sprite Size Mode=::NewEnumerator3
   )
   &ScaleMeshSize(Scale Factor=&Multiply_Vector(
     A=&VectorFromFloat(Value=&FloatFromCurve(
       FloatCurve=DataInterface() {propertyValues {Curve {
          defaultValue=1
          keys [
           <at 0 0 RCIM_Cubic>
           {arriveTangent=-0.1877, interpMode=RCIM_Cubic, leaveTangent=-0.1877, tangentMode=RCTM_User, time=0.5, value=1}
           {arriveTangent=0.0991, interpMode=RCIM_Cubic, leaveTangent=0.0991, tangentMode=RCTM_User, time=0.9, value=1}
           <at 1 0 RCIM_Cubic>
          ]
         }}}
       Scale Curve=$System.Timeline.InverseNormalizedTime
      ))
     B=&VectorFromCurve(VectorCurve=DataInterface() {propertyValues {XCurve {keys [<at 0.1 0>,<at 0.2 1>,<at 1 1>]}, YCurve {keys [<at 0 1>,<at 1 1>]}, ZCurve {keys [<at 0 1>,<at 1 1>]}}})
    ))
   &Drag(Drag=&RandomRangeFloat(Minimum=0.5, Maximum=10), Use Linear Drag=true, Use Rotational Drag=false)
   &CurlNoiseForce(Bias Noise=false, Noise Quality / Cost=::NewEnumerator0, Noise Strength=500, Pan Noise=false)
   &DynamicMaterialParameters(
    Index 1 Param 1=&FloatFromCurve(FloatCurve=DataInterface() {propertyValues {Curve {keys [<at 0 1>,<at 0.3 0>]}}})
    Index 1 Param 2=&RandomRangeFloat(Minimum=2, Maximum=3)
    Index 1 Param 3=&FloatFromCurve(
     FloatCurve=DataInterface() {propertyValues {Curve {keys [<at 0.2 5>,<at 0.3 5>,<at 1 0.5>]}}}
     Scale Curve=&FloatFromCurve(
      FloatCurve=DataInterface() {propertyValues {Curve {defaultValue=1, keys [{tangentMode=RCTM_User, time=0, value=0.3}, {tangentMode=RCTM_User, time=0.15, value=1}]}}}
      CurveIndex=$Particles.Age
     )
    )
    Index 1 Param 4=-1
    Index 2 Param 1=&RandomRangeFloat(Minimum=0, Maximum=1111)
    Index 3 Param 1=&RandomRangeFloat(Minimum=-0.5, Maximum=3)
    Index 3 Param 2=&RandomRangeFloat(Minimum=0, Maximum=1)
    Module.Index0Float0Write=true
    Module.Index0Float1Write=true
    Module.Index0Float2Write=false
    Module.Index0Float3Write=true
    Module.Index1Float0Write=true
    Module.Index1Float1Write=true
    Module.Index1Float2Write=true
    Module.Index1Float3Write=true
    Module.Index2Float0Write=true
    Module.Index2Float1Write=false
    Module.Index2Float2Write=false
    Module.Index2Float3Write=false
    Module.Index3Float0Write=true
    Module.Index3Float1Write=true
    Module.Index3Float2Write=true
    Module.Index3Float3Write=true
    Write Parameter Index 1=::NewEnumerator1
    Write Parameter Index 2=::NewEnumerator1
    Write Parameter Index 3=::NewEnumerator1
   )
   /SetVariables(Particles.ColorBlack:LinearColor=&MakeLinearColorFromVectorAndFloat(
     "Vector (RGB)"=&MakeVectorFromLinearColorRGB(Color=&ColorFromCurve(ColorCurve=DataInterface() {propertyValues {
         AlphaCurve {keys [<at 0 0 RCIM_Cubic>]}
         BlueCurve {keys [<at 0.745 0 RCIM_Cubic>]}
         GreenCurve {keys [<at 0.745 0 RCIM_Cubic>]}
         RedCurve {keys [<at 0.745 0 RCIM_Cubic>]}
        }}))
     "Float (Alpha)"=&MakeCustomFloatFromBool(
      Bool=$Particles.HasCollided
      True Float=0
      False Float=&FloatFromCurve(FloatCurve=DataInterface() {propertyValues {Curve {keys [<at 0.55 1>,<at 1 0>]}}}, Scale Curve=2)
     )
    )) {isSetParameters=true, module=SetVariables_#4}
   /SetVariables(
    Particles.DMP_Trail:Vector4f=&MakeVector4(
     X=0
     Y=-0.1
     Z=1
     W=&FloatFromCurve(FloatCurve=DataInterface() {propertyValues {Curve {defaultValue=1, keys [{tangentMode=RCTM_User, time=0.05, value=0}, {tangentMode=RCTM_User, time=1, value=1}]}}})
    )
    Particles.DMP_Trail1:Vector4f=&MakeVector4(X=1, Y=1, Z=1, W=1)
    Particles.DMP_Trail2:Vector4f=&MakeVector4(X=1, Y=1, Z=1, W=1)
    Particles.DMP_Trail3:Vector4f=&MakeVector4(X=500, Y=25, Z=0, W=0)
    Particles.Trail Color:LinearColor=&MultiplyLinearColor(
     Linear Color=&MakeCustomLinearColorFromBool(Bool=$User.Ascended, True Color=<rgba 2 0 0.9407 1>, False Color=<rgba 0.9219 0.024 0.024 1>)
     Scale Factor=&MakeVector4(
      X=1
      Y=1
      Z=1
      W=&MakeCustomFloatFromBool(
       Bool=$Particles.HasCollided
       True Float=0
       False Float=&FloatFromCurve(FloatCurve=DataInterface() {propertyValues {Curve {keys [<at 0.9 1>,<at 1 0>]}}}, Scale Curve=2)
      )
     )
    )
    Particles.Trail Orientation:Quat4f=&MakeQuaternion(
     Quaternion From=0
     Axis=<1 0 0>
     Angle=&Add_Float(
      A=&RandomRangeFloat(Minimum=0, Maximum=360)
      B=&Multiply_Float(A=$Engine.Time, B=1000)
     )
     From=?
     To=?
     Angle Type=::NewEnumerator0
     Normalize Vectors=false
    )
    Particles.Trail Scale:Vector3f=&VectorFromCurve(VectorCurve=DataInterface() {propertyValues {
       XCurve {defaultValue=1, keys [{tangentMode=RCTM_User, time=0, value=1.25}, {tangentMode=RCTM_User, time=0.2, value=2.375}, <at 1 0.5>]}
       YCurve {defaultValue=1, keys [{tangentMode=RCTM_User, time=0, value=0.25}, {tangentMode=RCTM_User, time=0.2, value=1}, <at 1 2>]}
       ZCurve {defaultValue=1, keys [{tangentMode=RCTM_User, time=0, value=0.25}, {tangentMode=RCTM_User, time=0.2, value=1}, <at 1 2>]}
      }})
    Particles.Velocity=&MakeCustomVectorFromBool(Bool=$Particles.HasCollided, True Vector=<0 0 0>, False Vector=$Particles.Velocity)
   ) {isSetParameters=true, module=SetVariables_#5}
   &Timeline(Play=$Particles.HasCollided, Play Rate=2)
   &ScaleColor(
    Scale RGB=&Multiply_VectorByFloat(
     Vector=&MakeVectorFromLinearColorRGB(Color=&ColorFromCurve(ColorCurve=DataInterface() {propertyValues {
         AlphaCurve {keys [<at 0 0.2>, {arriveTangent=-0.2, interpMode=RCIM_Cubic, leaveTangent=-0.2, time=0.1889, value=1}, <at 1 0 RCIM_Cubic>]}
         BlueCurve {keys [<at 0 1 RCIM_Cubic>]}
         GreenCurve {keys [<at 0 1 RCIM_Cubic>]}
         RedCurve {keys [<at 0 1 RCIM_Cubic>]}
        }}))
     Float=&MakeCustomFloatFromBool(Bool=$Particles.HasCollided, True Float=0, False Float=1)
    )
    ScaleA=false
    ScaleRGB=true
   )
   /SetVariables(
    Particles.Hit Color:LinearColor=&MakeLinearColorFromVectorAndFloat(
     "Vector (RGB)"=&MakeCustomVectorFromBool(
      Bool=$User.Ascended
      True Vector=&VectorFromCurve(
       VectorCurve=DataInterface() {propertyValues {
         XCurve {keys [{arriveTangent=-3.5433, time=0, value=0.625}, <at 0.5 0.1>]}
         YCurve {keys [<at 0.01 0.1>,<at 0.5 0>]}
         ZCurve {keys [<at 0 1.375>,<at 0.5 0.375>]}
        }}
       CurveIndex=$Particles.Timeline.NormalizedTime
      )
      False Vector=&VectorFromCurve(
       VectorCurve=DataInterface() {propertyValues {
         XCurve {keys [{arriveTangent=-12.0139, interpMode=RCIM_Cubic, leaveTangent=-12.0139, tangentMode=RCTM_User, time=0.01, value=2}, <at 0.5 0>]}
         YCurve {keys [<at 0.01 0.1>,<at 0.5 0>]}
         ZCurve {keys [<at 0.01 0.1>,<at 0.5 0>]}
        }}
       CurveIndex=$Particles.Timeline.NormalizedTime
      )
     )
     "Float (Alpha)"=&Multiply_Float(A=5, B=$Particles.Timeline.InverseNormalizedTime)
    )
    Particles.Hit Sprite:Vector2f=&Vector2DFromCurve(
     Vector2Curve=DataInterface() {propertyValues {
       XCurve {
        defaultValue=1
        keys [
         {arriveTangent=4.1785, interpMode=RCIM_Cubic, leaveTangent=4.1785, tangentMode=RCTM_User, time=0.01, value=0}
         {arriveTangent=1.1179, interpMode=RCIM_Cubic, leaveTangent=1.1179, tangentMode=RCTM_User, time=0.3, value=0.65}
         {arriveTangent=0.0224, interpMode=RCIM_Cubic, leaveTangent=0.0224, tangentMode=RCTM_User, time=1, value=1}
        ]
       }
       YCurve {
        defaultValue=1
        keys [
         {arriveTangent=4.1785, interpMode=RCIM_Cubic, leaveTangent=4.1785, tangentMode=RCTM_User, time=0.01, value=0}
         {arriveTangent=1.1179, interpMode=RCIM_Cubic, leaveTangent=1.1179, tangentMode=RCTM_User, time=0.3, value=0.65}
         {arriveTangent=0.0224, interpMode=RCIM_Cubic, leaveTangent=0.0224, tangentMode=RCTM_User, time=1, value=1}
        ]
       }
      }}
     CurveIndex=$Particles.Timeline.NormalizedTime
     ScaleCurve=<120 120>
    )
   ) {isSetParameters=true, module=SetVariables_#6}
   &SubUVAnimation(
    Random Start Frame=true
    "Sprite Renderer "=DataInterface() {propertyValues {SpriteRenderer=&self:Raven_Rotate_0.NiagaraSpriteRendererProperties_1}}
    SubUV Lookup Index:NiagaraFloat=$StackContext.Timeline.NormalizedTime
    UseEndFrame=false
    UseStartFrame=false
   )
   &VortexForce(
    Origin Pull Amount=2500
    UseOriginOffset=false
    Vortex Axis=<0 0 0>
    Vortex Force Amount=&RandomRangeFloat(Minimum=1250, Maximum=5000)
   )
   &PointAttractionForce(
    Attraction Radius=10000
    Attraction Strength=&RandomRangeFloat(
     Minimum=&FloatFromCurve(
      FloatCurve=DataInterface() {propertyValues {Curve {defaultValue=1, keys [<at 0.25 0 RCIM_Cubic>, {interpMode=RCIM_Cubic, leaveTangent=0.0388, time=1, value=1}]}}}
      Scale Curve=&RandomRangeFloat(Minimum=300, Maximum=1000)
     )
     Maximum=&FloatFromCurve(
      FloatCurve=DataInterface() {propertyValues {Curve {defaultValue=1, keys [<at 0.5 0 RCIM_Cubic>,<at 1 1 RCIM_Cubic>]}}}
      Scale Curve=&RandomRangeFloat(Minimum=500, Maximum=1000)
     )
    )
    Kill Within Radius=false
    Module.Use Falloff=true
    UseAttractorOffset=false
   ) {enabled=false}
   &SolveForcesAndVelocity(Clamp Velocity=false, Limit Acceleration=false)
  ]
  rendererClasses [&NiagaraMeshRendererProperties, &NiagaraSpriteRendererProperties]
  renderers [
   {rendererClass=&NiagaraMeshRendererProperties, rendererIndex=0}
   {rendererClass=&NiagaraSpriteRendererProperties, rendererIndex=1}
   {rendererClass=&NiagaraMeshRendererProperties, rendererIndex=2}
   {rendererClass=&NiagaraSpriteRendererProperties, rendererIndex=3}
   {rendererClass=&NiagaraMeshRendererProperties, rendererIndex=4}
  ]
 }
 Raven_Thrower {
  bEnabled=false
  emitterUpdateScript [
   &EmitterState()
   &SpawnRate(
    SpawnRate=&Multiply_Float(
     A=16.5
     B=&FloatFromCurve(FloatCurve=DataInterface() {propertyValues {Curve {keys [<at 0.5 1 RCIM_Cubic>,<at 1 0 RCIM_Cubic>]}}}, CurveIndex=$System.NormalizedLoopAge)
    )
    Use Spawn Probability=false
   )
  ]
  particleSpawnScript [
   &InitializeParticle(
    Color=&MakeCustomLinearColorFromBool(Bool=$User.Ascended, True Color=<rgba 0.7481 0 2.5 1>, False Color=<rgba 5 0 0.2 1>)
    Lifetime=&RandomRangeFloat(Minimum=2, Maximum=4)
    Mesh Scale Mode=::NewEnumerator3
    Mesh Uniform Scale=&RandomRangeFloat(Minimum=1.5, Maximum=3)
    Sprite Size Mode=::NewEnumerator3
    Uniform Sprite Size=&MakeCustomFloatFromBool(Bool=$User.Ascended, True Float=250, False Float=150)
    UsePositionOffset=false
    Write Material Random=true
   )
   &ShapeLocation(
    MaskSpawnGroup=false
    Non Uniform Scale=<1 1 0.5>
    Offset Coordinate Space=::Local
    Offset Mode=::NewEnumerator0
    Override Seed=false
    Rotation Axis=<0 0 0>
    Rotation Mode=::NewEnumerator1
    Sphere Radius=7000
    Sphere Surface Distribution=1
    UseOwnerScale=true
   )
   &InitialMeshOrientation(Mesh Orientation Mode=::NewEnumerator2, Use Rotation Vector=false)
   &AddVelocity(
    Module.Use Falloff Distance=false
    Override Seed=false
    Velocity Mode=::NewEnumerator1
    Velocity Speed=&Multiply_Float(A=-250, B=$User.Velocity Scale)
   ) {enabled=false}
   &CameraOffset(Camera Offset Amount=-25)
   /SetVariables(Particles.CamOfs2=30) {isSetParameters=true, module=SetVariables_#3}
  ]
  particleUpdateScript [
   &ParticleState()
   &ScaleSpriteSize(
    Non-Uniform Curve Scale=<1.5 1>
    Non-Uniform Curve Sprite Scale=DataInterface() {propertyValues {
      XCurve {keys [<at 0.05 0 RCIM_Cubic>, {tangentMode=RCTM_User, time=0.3, value=1}, <at 0.5 1>, <at 1 0>]}
      YCurve {keys [<at 0.05 0 RCIM_Cubic>, {tangentMode=RCTM_User, time=0.3, value=1}, <at 0.5 1>, <at 1 0>]}
     }}
    Scale Sprite Size Mode=::NewEnumerator3
   )
   &ScaleMeshSize(Scale Factor=&Multiply_Vector(
     A=&VectorFromFloat(Value=&FloatFromCurve(
       FloatCurve=DataInterface() {propertyValues {Curve {
          defaultValue=1
          keys [
           <at 0 0 RCIM_Cubic>
           {arriveTangent=-0.1877, interpMode=RCIM_Cubic, leaveTangent=-0.1877, tangentMode=RCTM_User, time=0.5, value=1}
           {arriveTangent=0.0991, interpMode=RCIM_Cubic, leaveTangent=0.0991, tangentMode=RCTM_User, time=0.9, value=1}
           <at 1 0 RCIM_Cubic>
          ]
         }}}
       Scale Curve=$System.Timeline.InverseNormalizedTime
      ))
     B=&VectorFromCurve(VectorCurve=DataInterface() {propertyValues {XCurve {keys [<at 0.1 0>,<at 0.2 1>,<at 1 1>]}, YCurve {keys [<at 0 1>,<at 1 1>]}, ZCurve {keys [<at 0 1>,<at 1 1>]}}})
    ))
   &Drag(Drag=&RandomRangeFloat(Minimum=0.5, Maximum=10), Use Linear Drag=true, Use Rotational Drag=false)
   &CurlNoiseForce(Bias Noise=false, Noise Quality / Cost=::NewEnumerator0, Noise Strength=500, Pan Noise=false)
   &DynamicMaterialParameters(
    Index 1 Param 1=&FloatFromCurve(FloatCurve=DataInterface() {propertyValues {Curve {keys [<at 0 1>,<at 0.3 0>]}}})
    Index 1 Param 2=&RandomRangeFloat(Minimum=2, Maximum=3)
    Index 1 Param 3=&FloatFromCurve(
     FloatCurve=DataInterface() {propertyValues {Curve {keys [<at 0.2 5>,<at 0.3 5>,<at 1 0.5>]}}}
     Scale Curve=&FloatFromCurve(
      FloatCurve=DataInterface() {propertyValues {Curve {defaultValue=1, keys [{tangentMode=RCTM_User, time=0, value=0.3}, {tangentMode=RCTM_User, time=0.15, value=1}]}}}
      CurveIndex=$Particles.Age
     )
    )
    Index 1 Param 4=-1
    Index 2 Param 1=&RandomRangeFloat(Minimum=0, Maximum=1111)
    Index 3 Param 1=&RandomRangeFloat(Minimum=-0.5, Maximum=3)
    Index 3 Param 2=&RandomRangeFloat(Minimum=0, Maximum=1)
    Module.Index0Float0Write=true
    Module.Index0Float1Write=true
    Module.Index0Float2Write=false
    Module.Index0Float3Write=true
    Module.Index1Float0Write=true
    Module.Index1Float1Write=true
    Module.Index1Float2Write=true
    Module.Index1Float3Write=true
    Module.Index2Float0Write=true
    Module.Index2Float1Write=false
    Module.Index2Float2Write=false
    Module.Index2Float3Write=false
    Module.Index3Float0Write=true
    Module.Index3Float1Write=true
    Module.Index3Float2Write=true
    Module.Index3Float3Write=true
    Write Parameter Index 1=::NewEnumerator1
    Write Parameter Index 2=::NewEnumerator1
    Write Parameter Index 3=::NewEnumerator1
   )
   /SetVariables(Particles.ColorBlack:LinearColor=&MakeLinearColorFromVectorAndFloat(
     "Vector (RGB)"=&MakeVectorFromLinearColorRGB(Color=&ColorFromCurve(ColorCurve=DataInterface() {propertyValues {
         AlphaCurve {keys [<at 0 0 RCIM_Cubic>]}
         BlueCurve {keys [<at 0.745 0 RCIM_Cubic>]}
         GreenCurve {keys [<at 0.745 0 RCIM_Cubic>]}
         RedCurve {keys [<at 0.745 0 RCIM_Cubic>]}
        }}))
     "Float (Alpha)"=&MakeCustomFloatFromBool(
      Bool=$Particles.HasCollided
      True Float=0
      False Float=&FloatFromCurve(FloatCurve=DataInterface() {propertyValues {Curve {keys [<at 0.55 1>,<at 1 0>]}}}, Scale Curve=2)
     )
    )) {isSetParameters=true, module=SetVariables_#4}
   /SetVariables(
    Particles.DMP_Trail:Vector4f=&MakeVector4(
     X=0
     Y=-0.1
     Z=1
     W=&FloatFromCurve(FloatCurve=DataInterface() {propertyValues {Curve {defaultValue=1, keys [{tangentMode=RCTM_User, time=0.05, value=0}, {tangentMode=RCTM_User, time=1, value=1}]}}})
    )
    Particles.DMP_Trail1:Vector4f=&MakeVector4(X=1, Y=1, Z=1, W=1)
    Particles.DMP_Trail2:Vector4f=&MakeVector4(X=1, Y=1, Z=1, W=1)
    Particles.DMP_Trail3:Vector4f=&MakeVector4(X=500, Y=25, Z=0, W=0)
    Particles.Trail Color:LinearColor=&MultiplyLinearColor(
     Linear Color=&MakeCustomLinearColorFromBool(Bool=$User.Ascended, True Color=<rgba 2 0 0.9407 1>, False Color=<rgba 0.9219 0.024 0.024 1>)
     Scale Factor=&MakeVector4(
      X=1
      Y=1
      Z=1
      W=&MakeCustomFloatFromBool(
       Bool=$Particles.HasCollided
       True Float=0
       False Float=&FloatFromCurve(FloatCurve=DataInterface() {propertyValues {Curve {keys [<at 0.9 1>,<at 1 0>]}}}, Scale Curve=2)
      )
     )
    )
    Particles.Trail Orientation:Quat4f=&MakeQuaternion(
     Quaternion From=0
     Axis=<1 0 0>
     Angle=&Add_Float(
      A=&RandomRangeFloat(Minimum=0, Maximum=360)
      B=&Multiply_Float(A=$Engine.Time, B=1000)
     )
     From=?
     To=?
     Angle Type=::NewEnumerator0
     Normalize Vectors=false
    )
    Particles.Trail Scale:Vector3f=&VectorFromCurve(VectorCurve=DataInterface() {propertyValues {
       XCurve {defaultValue=1, keys [{tangentMode=RCTM_User, time=0, value=1.25}, {tangentMode=RCTM_User, time=0.2, value=2.375}, <at 1 0.5>]}
       YCurve {defaultValue=1, keys [{tangentMode=RCTM_User, time=0, value=0.25}, {tangentMode=RCTM_User, time=0.2, value=1}, <at 1 2>]}
       ZCurve {defaultValue=1, keys [{tangentMode=RCTM_User, time=0, value=0.25}, {tangentMode=RCTM_User, time=0.2, value=1}, <at 1 2>]}
      }})
    Particles.Velocity=&MakeCustomVectorFromBool(Bool=$Particles.HasCollided, True Vector=<0 0 0>, False Vector=$Particles.Velocity)
   ) {isSetParameters=true, module=SetVariables_#5}
   &Timeline(Play=$Particles.HasCollided, Play Rate=2)
   &ScaleColor(
    Scale RGB=&Multiply_VectorByFloat(
     Vector=&MakeVectorFromLinearColorRGB(Color=&ColorFromCurve(ColorCurve=DataInterface() {propertyValues {
         AlphaCurve {keys [<at 0 0.2>, {arriveTangent=-0.2, interpMode=RCIM_Cubic, leaveTangent=-0.2, time=0.1889, value=1}, <at 1 0 RCIM_Cubic>]}
         BlueCurve {keys [<at 0 1 RCIM_Cubic>]}
         GreenCurve {keys [<at 0 1 RCIM_Cubic>]}
         RedCurve {keys [<at 0 1 RCIM_Cubic>]}
        }}))
     Float=&MakeCustomFloatFromBool(Bool=$Particles.HasCollided, True Float=0, False Float=1)
    )
    ScaleA=false
    ScaleRGB=true
   )
   /SetVariables(
    Particles.Hit Color:LinearColor=&MakeLinearColorFromVectorAndFloat(
     "Vector (RGB)"=&MakeCustomVectorFromBool(
      Bool=$User.Ascended
      True Vector=&VectorFromCurve(
       VectorCurve=DataInterface() {propertyValues {
         XCurve {keys [{arriveTangent=-3.5433, time=0, value=0.625}, <at 0.5 0.1>]}
         YCurve {keys [<at 0.01 0.1>,<at 0.5 0>]}
         ZCurve {keys [<at 0 1.375>,<at 0.5 0.375>]}
        }}
       CurveIndex=$Particles.Timeline.NormalizedTime
      )
      False Vector=&VectorFromCurve(
       VectorCurve=DataInterface() {propertyValues {
         XCurve {keys [{arriveTangent=-12.0139, interpMode=RCIM_Cubic, leaveTangent=-12.0139, tangentMode=RCTM_User, time=0.01, value=2}, <at 0.5 0>]}
         YCurve {keys [<at 0.01 0.1>,<at 0.5 0>]}
         ZCurve {keys [<at 0.01 0.1>,<at 0.5 0>]}
        }}
       CurveIndex=$Particles.Timeline.NormalizedTime
      )
     )
     "Float (Alpha)"=&Multiply_Float(A=5, B=$Particles.Timeline.InverseNormalizedTime)
    )
    Particles.Hit Sprite:Vector2f=&Vector2DFromCurve(
     Vector2Curve=DataInterface() {propertyValues {
       XCurve {
        defaultValue=1
        keys [
         {arriveTangent=4.1785, interpMode=RCIM_Cubic, leaveTangent=4.1785, tangentMode=RCTM_User, time=0.01, value=0}
         {arriveTangent=1.1179, interpMode=RCIM_Cubic, leaveTangent=1.1179, tangentMode=RCTM_User, time=0.3, value=0.65}
         {arriveTangent=0.0224, interpMode=RCIM_Cubic, leaveTangent=0.0224, tangentMode=RCTM_User, time=1, value=1}
        ]
       }
       YCurve {
        defaultValue=1
        keys [
         {arriveTangent=4.1785, interpMode=RCIM_Cubic, leaveTangent=4.1785, tangentMode=RCTM_User, time=0.01, value=0}
         {arriveTangent=1.1179, interpMode=RCIM_Cubic, leaveTangent=1.1179, tangentMode=RCTM_User, time=0.3, value=0.65}
         {arriveTangent=0.0224, interpMode=RCIM_Cubic, leaveTangent=0.0224, tangentMode=RCTM_User, time=1, value=1}
        ]
       }
      }}
     CurveIndex=$Particles.Timeline.NormalizedTime
     ScaleCurve=<120 120>
    )
   ) {isSetParameters=true, module=SetVariables_#6}
   &SubUVAnimation(
    Random Start Frame=true
    "Sprite Renderer "=DataInterface() {propertyValues {SpriteRenderer=&self:Raven_Thrower_0.NiagaraSpriteRendererProperties_1}}
    SubUV Lookup Index:NiagaraFloat=$StackContext.Timeline.NormalizedTime
    UseEndFrame=false
    UseStartFrame=false
   )
   &VortexForce(
    UseOriginOffset=false
    Vortex Axis=<0 0 0>
    Vortex Force Amount=&RandomRangeFloat(Minimum=2500, Maximum=10000)
   )
   &PointAttractionForce(
    Attraction Radius=10000
    Attraction Strength=&RandomRangeFloat(
     Minimum=&FloatFromCurve(
      FloatCurve=DataInterface() {propertyValues {Curve {defaultValue=1, keys [<at 0.25 0 RCIM_Cubic>, {interpMode=RCIM_Cubic, leaveTangent=0.0388, time=1, value=1}]}}}
      Scale Curve=&RandomRangeFloat(Minimum=300, Maximum=1000)
     )
     Maximum=&FloatFromCurve(
      FloatCurve=DataInterface() {propertyValues {Curve {defaultValue=1, keys [<at 0.5 0 RCIM_Cubic>,<at 1 1 RCIM_Cubic>]}}}
      Scale Curve=&RandomRangeFloat(Minimum=500, Maximum=1000)
     )
    )
    Kill Within Radius=false
    Module.Use Falloff=true
    UseAttractorOffset=false
   )
   &SolveForcesAndVelocity(Clamp Velocity=false, Limit Acceleration=false)
  ]
  rendererClasses [&NiagaraMeshRendererProperties, &NiagaraSpriteRendererProperties]
  renderers [
   {rendererClass=&NiagaraMeshRendererProperties, rendererIndex=0}
   {rendererClass=&NiagaraSpriteRendererProperties, rendererIndex=1}
   {rendererClass=&NiagaraMeshRendererProperties, rendererIndex=2}
   {rendererClass=&NiagaraSpriteRendererProperties, rendererIndex=3}
   {rendererClass=&NiagaraMeshRendererProperties, rendererIndex=4}
  ]
 }
}
summary {
 systemName=NS_MistRaven_DarkCoil_RavenThrower_AttackCinematic
 emitters [
  {
   bEnabled=false
   emitterName=Raven_Thrower
   rendererClasses [&NiagaraMeshRendererProperties, &NiagaraSpriteRendererProperties]
  }
  {emitterName=Impact, rendererClasses [&NiagaraMeshRendererProperties, &NiagaraSpriteRendererProperties]}
  {emitterName=Raven_Rotate, rendererClasses [&NiagaraMeshRendererProperties, &NiagaraSpriteRendererProperties]}
  {emitterName=Raven_Attack, rendererClasses [&NiagaraMeshRendererProperties, &NiagaraSpriteRendererProperties]}
  {bEnabled=false, emitterName=Minimal, rendererClasses [&NiagaraSpriteRendererProperties]}
  {bEnabled=false, emitterName=Raven_Attack001, rendererClasses [&NiagaraMeshRendererProperties]}
 ]
 userVariables [
  User.ActiveTarget {type=NiagaraBool}
  User.Ascended {type=NiagaraBool, defaultValue=true}
  User.Cone Angle {type=NiagaraFloat, defaultValue=20, description=""}
  User.Culling Alpha Power {type=NiagaraFloat, defaultValue=8, description=""}
  User.Duration {type=NiagaraFloat, defaultValue=4, description=""}
  User.Raven Kill Distance {type=NiagaraFloat, defaultValue=250, description=""}
  User.Spline Thicken {type=NiagaraFloat, defaultValue=100, description=""}
  User.TargetVector {type=Vector3f, description="", defaultValue {struct=Vector3f, value=<1 0.075 0.02>}}
  User.Velocity Scale {type=NiagaraFloat, defaultValue=30, description=""}
 ]
}
